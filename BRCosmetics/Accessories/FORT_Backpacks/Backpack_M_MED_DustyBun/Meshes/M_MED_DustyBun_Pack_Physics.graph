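# Physics asset for the DustyBun backpack cosmetic, defining simulation bodies and joint limits for its paired wings and ponytails.
# Each wing master hangs from body_master and branches into lower and mid sections, with a six-segment ponytail chain off each lower base.
# Tight angular limits generally loosen along the chains, while two extra angularly free constraints cross-link left and right ponytail segments 03 and 04.
# A large, unconstrained pelvis capsule sits alongside the rig.
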PhysicsAsset /BRCosmetics/Accessories/FORT_Backpacks/Backpack_M_MED_DustyBun/Meshes/M_MED_DustyBun_Pack_Physics
bodies {
 Spelvis [{shapeType=Capsule, length=58.3709, radius=21.1795, shapeName=Spelvis_capsule, center=<52.7152 1.6162 -36.4679>, rotation=<rot 90 -90 0>}]
 body_master [{shapeType=Capsule, length=17.18, radius=5, shapeName=body_master_capsule, center=<1.8971 0 0>, rotation=<rot 90 0 0>}]
 wing_l_lower_base_jnt [{shapeType=Box, extentX=20.7752, extentY=2.7567, extentZ=26.979, shapeName=wing_l_lower_base_jnt_box, center=<9.3206 0 0>}]
 wing_l_lower_out_01_jnt [{shapeType=Capsule, length=20.6484, radius=4.775, shapeName=wing_l_lower_out_01_jnt_capsule, center=<2.0088 0.0261 0>, rotation=<rot 90 0 0>}]
 wing_l_master_jnt [{shapeType=Capsule, length=20.1476, radius=5, shapeName=wing_l_master_jnt_capsule, center=<4.3977 0 0>}]
 wing_l_mid_base_jnt [{
   shapeType=Capsule
   length=24.2254
   radius=6.3505
   shapeName=wing_l_mid_base_jnt_capsule
   center=<15.8536 1.3934 -6.8439>
   rotation=<rot -82.1158 -157.5315 172.3097>
  }]
 wing_l_mid_out_01_jnt [{shapeType=Capsule, length=10.4076, radius=9.0796, shapeName=wing_l_mid_out_01_jnt_capsule, center=<14.6683 0 0>, rotation=<rot 90 0 0>}]
 wing_l_mid_out_02_jnt [{shapeType=Capsule, length=12.2947, radius=4.3297, shapeName=wing_l_mid_out_02_jnt_capsule, center=<0.4002 -4.7747 0>, rotation=<rot 90 0 0>}]
 wing_l_ponytail_01_jnt [{shapeType=Capsule, length=11.181, radius=3.819, shapeName=wing_l_ponytail_01_jnt_capsule, center=<5.2335 0 0>, rotation=<rot 90 0 0>}]
 wing_l_ponytail_02_jnt [{shapeType=Capsule, length=16.1211, radius=6.844, shapeName=wing_l_ponytail_02_jnt_capsule, center=<10.0497 0 0>, rotation=<rot 90 0 0>}]
 wing_l_ponytail_03_jnt [{shapeType=Capsule, length=20.1652, radius=8.6161, shapeName=wing_l_ponytail_03_jnt_capsule, center=<16.6271 0 0>, rotation=<rot 90 0 0>}]
 wing_l_ponytail_04_jnt [{shapeType=Capsule, length=8.3573, radius=6.3406, shapeName=wing_l_ponytail_04_jnt_capsule, center=<5.2335 0 0>, rotation=<rot 90 0 0>}]
 wing_l_ponytail_05_jnt [{shapeType=Capsule, length=8.3573, radius=6.3406, shapeName=wing_l_ponytail_05_jnt_capsule, center=<5.2335 0 0>, rotation=<rot 90 0 0>}]
 wing_l_ponytail_06_jnt [{shapeType=Capsule, length=8.3573, radius=6.3406, shapeName=wing_l_ponytail_06_jnt_capsule, center=<5.2335 0 0>, rotation=<rot 90 0 0>}]
 wing_r_lower_base_jnt [{shapeType=Box, extentX=20.7752, extentY=2.7567, extentZ=20.5086, shapeName=wing_r_lower_base_jnt_box, center=<9.3206 0 0>}]
 wing_r_lower_out_01_jnt [{shapeType=Capsule, length=20.4208, radius=4.5787, shapeName=wing_r_lower_out_01_jnt_capsule, center=<0.8913 4.7532 0>, rotation=<rot 90 0 0>}]
 wing_r_master_jnt [{shapeType=Capsule, length=20.1476, radius=5, shapeName=wing_r_master_jnt_capsule, center=<4.3977 0 0>}]
 wing_r_mid_base_jnt [{
   shapeType=Capsule
   length=22.9768
   radius=5.9807
   shapeName=wing_r_mid_base_jnt_capsule
   center=<15.2073 -1.0738 -4.6229>
   rotation=<rot 90 -16.8341 0>
  }]
 wing_r_mid_out_01_jnt [{shapeType=Capsule, length=10.6755, radius=9.3456, shapeName=wing_r_mid_out_01_jnt_capsule, center=<14.5476 0 0>, rotation=<rot 90 0 0>}]
 wing_r_mid_out_02_jnt [{shapeType=Capsule, length=19.0099, radius=4.3597, shapeName=wing_r_mid_out_02_jnt_capsule, center=<4.1855 2.5378 0>, rotation=<rot 90 0 0>}]
 wing_r_ponytail_01_jnt [{shapeType=Capsule, length=14.9774, radius=3.5226, shapeName=wing_r_ponytail_01_jnt_capsule, center=<6.7935 0 0>, rotation=<rot 90 0 0>}]
 wing_r_ponytail_02_jnt [{shapeType=Capsule, length=16.7624, radius=6.4637, shapeName=wing_r_ponytail_02_jnt_capsule, center=<10.4346 0 0>, rotation=<rot 90 0 0>}]
 wing_r_ponytail_03_jnt [{shapeType=Capsule, length=18.9871, radius=8.0237, shapeName=wing_r_ponytail_03_jnt_capsule, center=<15.9413 0 0>, rotation=<rot 90 0 0>}]
 wing_r_ponytail_04_jnt [{shapeType=Capsule, length=8.5208, radius=6.1771, shapeName=wing_r_ponytail_04_jnt_capsule, center=<5.1874 0 0>, rotation=<rot 90 0 0>}]
 wing_r_ponytail_05_jnt [{shapeType=Capsule, length=8.5208, radius=6.1771, shapeName=wing_r_ponytail_05_jnt_capsule, center=<5.1874 0 0>, rotation=<rot 90 0 0>}]
 wing_r_ponytail_06_jnt [{shapeType=Capsule, length=8.5208, radius=6.1771, shapeName=wing_r_ponytail_06_jnt_capsule, center=<5.1874 0 0>, rotation=<rot 90 0 0>}]
}
constraints [
 {bone1Name=wing_l_lower_base_jnt, bone2Name=wing_l_master_jnt, swing1LimitDegrees=0, swing2LimitDegrees=0, twistLimitDegrees=0}
 {bone1Name=wing_l_master_jnt, bone2Name=body_master, swing1LimitDegrees=10, swing2LimitDegrees=0, twistLimitDegrees=0}
 {bone1Name=wing_l_lower_out_01_jnt, bone2Name=wing_l_lower_base_jnt, swing1LimitDegrees=5, swing2LimitDegrees=5, twistLimitDegrees=0}
 {bone1Name=wing_l_ponytail_01_jnt, bone2Name=wing_l_lower_base_jnt, swing1LimitDegrees=10, swing2LimitDegrees=0, twistLimitDegrees=0}
 {bone1Name=wing_l_ponytail_02_jnt, bone2Name=wing_l_ponytail_01_jnt, swing1LimitDegrees=10, swing2LimitDegrees=10, twistLimitDegrees=0}
 {bone1Name=wing_l_ponytail_03_jnt, bone2Name=wing_l_ponytail_02_jnt, swing1LimitDegrees=15, swing2LimitDegrees=10, twistLimitDegrees=15}
 {bone1Name=wing_l_ponytail_04_jnt, bone2Name=wing_l_ponytail_03_jnt, swing1LimitDegrees=15, swing2LimitDegrees=15, twistLimitDegrees=15}
 {bone1Name=wing_l_ponytail_05_jnt, bone2Name=wing_l_ponytail_04_jnt, swing1LimitDegrees=15, swing2LimitDegrees=15, twistLimitDegrees=15}
 {bone1Name=wing_l_ponytail_06_jnt, bone2Name=wing_l_ponytail_05_jnt, swing1LimitDegrees=15, swing2LimitDegrees=15, twistLimitDegrees=15}
 {bone1Name=wing_l_mid_base_jnt, bone2Name=wing_l_master_jnt, swing1LimitDegrees=10, swing2LimitDegrees=5, twistLimitDegrees=0}
 {bone1Name=wing_l_mid_out_01_jnt, bone2Name=wing_l_mid_base_jnt, swing1LimitDegrees=20, swing2LimitDegrees=5, twistLimitDegrees=0}
 {bone1Name=wing_l_mid_out_02_jnt, bone2Name=wing_l_mid_base_jnt, swing1LimitDegrees=20, swing2LimitDegrees=5, twistLimitDegrees=0}
 {bone1Name=wing_r_master_jnt, bone2Name=body_master, swing1LimitDegrees=10, swing2LimitDegrees=0, twistLimitDegrees=0}
 {bone1Name=wing_r_lower_base_jnt, bone2Name=wing_r_master_jnt, swing1LimitDegrees=0, swing2LimitDegrees=0, twistLimitDegrees=0}
 {bone1Name=wing_r_lower_out_01_jnt, bone2Name=wing_r_lower_base_jnt, swing1LimitDegrees=5, swing2LimitDegrees=5, twistLimitDegrees=0}
 {bone1Name=wing_r_ponytail_01_jnt, bone2Name=wing_r_lower_base_jnt, swing1LimitDegrees=10, swing2LimitDegrees=0, twistLimitDegrees=0}
 {bone1Name=wing_r_ponytail_02_jnt, bone2Name=wing_r_ponytail_01_jnt, swing1LimitDegrees=10, swing2LimitDegrees=10, twistLimitDegrees=0}
 {bone1Name=wing_r_ponytail_03_jnt, bone2Name=wing_r_ponytail_02_jnt, swing1LimitDegrees=10, swing2LimitDegrees=10, twistLimitDegrees=15}
 {bone1Name=wing_r_ponytail_04_jnt, bone2Name=wing_r_ponytail_03_jnt, swing1LimitDegrees=10, swing2LimitDegrees=15, twistLimitDegrees=15}
 {bone1Name=wing_r_ponytail_05_jnt, bone2Name=wing_r_ponytail_04_jnt, swing1LimitDegrees=10, swing2LimitDegrees=15, twistLimitDegrees=15}
 {bone1Name=wing_r_ponytail_06_jnt, bone2Name=wing_r_ponytail_05_jnt, swing1LimitDegrees=10, swing2LimitDegrees=15, twistLimitDegrees=15}
 {bone1Name=wing_r_mid_base_jnt, bone2Name=wing_r_master_jnt, swing1LimitDegrees=10, swing2LimitDegrees=5, twistLimitDegrees=0}
 {bone1Name=wing_r_mid_out_01_jnt, bone2Name=wing_r_mid_base_jnt, swing1LimitDegrees=20, swing2LimitDegrees=5, twistLimitDegrees=0}
 {bone1Name=wing_r_mid_out_02_jnt, bone2Name=wing_r_mid_base_jnt, swing1LimitDegrees=20, swing2LimitDegrees=5, twistLimitDegrees=0}
 {
  bone1Name=wing_l_ponytail_03_jnt
  bone2Name=wing_r_ponytail_03_jnt
  swing1LimitDegrees=45
  swing1Motion=Free
  swing2LimitDegrees=45
  swing2Motion=Free
  twistLimitDegrees=45
  twistMotion=Free
 }
 {
  bone1Name=wing_l_ponytail_04_jnt
  bone2Name=wing_r_ponytail_04_jnt
  swing1LimitDegrees=45
  swing1Motion=Free
  swing2LimitDegrees=45
  swing2Motion=Free
  twistLimitDegrees=45
  twistMotion=Free
 }
]
