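# Animation Blueprint for the DustyBun backpack, posing its wings and pack from the wearer's movement and flight state.
# The event graph derives movement alphas, bone-driven rotations, glider state and gravity from the player.
# The skydiving branch combines RigidBody physics with additive bone offsets; the other uses front-end-switched RigidBody, world-space pelvis pinning and directional wing leans.
# The cached backward, crouch and melee chain (CombAnimCache) never reaches the output.
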AnimBlueprint /BRCosmetics/Accessories/FORT_Backpacks/Backpack_M_MED_DustyBun/Meshes/M_MED_DustyBun_Pack_AnimBP
target_skeleton=&/M_MED_DustyBun_Pack_Skeleton
graphs [
 AnimGraph {nodes [
   n1 = Get bRigidBodyEnabled()
   n2 = "Transform (Modify) Bone - Bone: wing_l_master_jnt"!(ComponentPose, bAlphaBoolEnabled:bool=@n1, Translation="0,0,0", Rotation="15,0,0", Scale="1,1,1") -> (Pose) {
    alphaInputType=Bool
    rotationMode=BMM_Additive
    alphaBoolBlend {blendOption=Linear}
    boneToModify {boneName=wing_l_master_jnt}
    rotation {pitch=15}
   }
   "Transform (Modify) Bone - Bone: wing_r_master_jnt"!(ComponentPose=@n2, bAlphaBoolEnabled:bool=@n1, Translation="0,0,0", Rotation="-15,0,0", Scale="1,1,1") -> (Pose) {
    alphaInputType=Bool
    rotationMode=BMM_Additive
    alphaBoolBlend {blendOption=Linear}
    boneToModify {boneName=wing_r_master_jnt}
    rotation {pitch=-15}
   }
   n3 = "Transform (Modify) Bone - Bone: wing_l_master_jnt"!(
    ComponentPose=Local To Component(LocalPose=Use cached pose 'toInput'())
    Alpha:float=1
    Translation="0,0,0"
    Rotation="20,0,0"
    Scale="1,1,1"
   ) -> (Pose) {rotationMode=BMM_Additive, alphaBoolBlend {blendOption=Linear}, boneToModify {boneName=wing_l_master_jnt}, rotation {pitch=20}}
   n4 = "Transform (Modify) Bone - Bone: wing_r_master_jnt"!(ComponentPose=@n3, Alpha:float=1, Translation="0,0,0", Rotation="-20,0,0", Scale="1,1,1") -> (Pose) {rotationMode=BMM_Additive, alphaBoolBlend {blendOption=Linear}, boneToModify {boneName=wing_r_master_jnt}, rotation {pitch=-20}}
   n5 = "Transform (Modify) Bone - Bone: body_master"(
    ComponentPose=@n4
    Alpha=Get Parashuting() -> (Parashuting:double)
    Translation="-16.495,-0.216,-1.035"
    Rotation="40.000,180.000,-180.000"
    Scale="1,1,1"
   ) {
    rotationMode=BMM_Additive
    rotationSpace=BCS_BoneSpace
    scaleMode=BMM_Additive
    scaleSpace=BCS_BoneSpace
    translationMode=BMM_Additive
    translationSpace=BCS_BoneSpace
    alphaBoolBlend {blendOption=Linear}
    alphaScaleBiasClamp {bInterpResult=true}
    boneToModify {boneName=body_master}
    rotation=<rot 40.0001 180 -180>
    translation=<-16.495 -0.216 -1.035>
   }
   n6 = "Transform (Modify) Bone - Bone: pack"(ComponentPose=@n5, Alpha=Get Skydive() -> (Skydive:double), Translation="0,0,0", Rotation="0,0,15") {
    rotationMode=BMM_Additive
    rotationSpace=BCS_BoneSpace
    scaleMode=BMM_Additive
    scaleSpace=BCS_BoneSpace
    translationMode=BMM_Additive
    translationSpace=BCS_BoneSpace
    alphaBoolBlend {blendOption=Linear}
    alphaScaleBiasClamp {bInterpResult=true}
    boneToModify {boneName=pack}
    rotation {roll=15}
   }
   n7 = RigidBody!(
    ComponentPose=@n6
    Alpha=Get alphaSkyCond() -> (alphaSkyCond:double)
    OverrideWorldGravity:Vector=Get Override World Gravity() -> (Override World Gravity:Vector)
    ExternalForce="1.5,1.5,1.5"
    ComponentLinearAccScale:Vector=Get Component Linear Acc Scale() -> (Component Linear Acc Scale:Vector)
   ) -> (Pose) {
    bOverrideWorldGravity=true
    overridePhysicsAsset=&/M_MED_DustyBun_Pack_Physics
    simulationSpace=BaseBoneSpace
    alphaBoolBlend {blendOption=Linear}
    alphaScaleBiasClamp {bInterpResult=true}
    baseBoneRef {boneName=pack}
    componentAppliedLinearAccClamp=<1000 1000 1000>
    componentLinearAccScale=<0.1 0.1 0.1>
    componentLinearVelScale=<0.1 0.1 0.1>
    overrideWorldGravity {z=-980}
    simSpaceSettings {maxAngularAcceleration=1300, maxAngularVelocity=150, maxLinearAcceleration=1000, maxLinearVelocity=600, velocityScaleZ=0.5, worldAlpha=1}
   }
   n8 = "Transform (Modify) Bone - Bone: wing_l_master_jnt"(ComponentPose=@n7, Alpha=1, Translation="0,0,0", Rotation=Get Return Value Rotation l() -> (Return Value Rotation l:Rotator), Scale="1,1,1") {
    rotationMode=BMM_Additive
    translationMode=BMM_Additive
    alphaBoolBlend {blendOption=Linear}
    alphaScaleBiasClamp {bInterpResult=true, scale=0.25}
    boneToModify {boneName=wing_l_master_jnt}
   }
   n9 = "Transform (Modify) Bone - Bone: wing_r_master_jnt"(ComponentPose=@n8, Alpha=1, Translation="0,0,0", Rotation=Get Return Value Rotation_r() -> (Return Value Rotation_r:Rotator), Scale="1,1,1") {
    rotationMode=BMM_Additive
    translationMode=BMM_Additive
    alphaBoolBlend {blendOption=Linear}
    alphaScaleBiasClamp {bInterpResult=true, scale=0.5}
    boneToModify {boneName=wing_r_master_jnt}
   }
   n10 = "Transform (Modify) Bone - Bone: wing_l_master_jnt"!(
    ComponentPose=@n9
    Alpha=1
    Translation="0,0,0"
    Rotation_Roll:float=0
    Rotation_Pitch:float=0
    Rotation_Yaw:float="Get Y (Pitch)UpperArmL"() -> ("Y (Pitch)UpperArmL":float)
    Scale="1,1,1"
   ) -> (Pose) {
    rotationMode=BMM_Additive
    translationMode=BMM_Additive
    alphaBoolBlend {blendOption=Linear}
    alphaScaleBias {scale=2}
    alphaScaleBiasClamp {bInterpResult=true, scale=0.25}
    boneToModify {boneName=wing_l_master_jnt}
   }
   n11 = "Transform (Modify) Bone - Bone: wing_l_master_jnt"(
    ComponentPose=@n10
    Alpha=1
    Translation="0,0,0"
    Rotation_Roll=0
    Rotation_Pitch=15
    Rotation_Yaw=Get rotValue_FlightState_L() -> (rotValue_FlightState_L:double)
    Scale="1,1,1"
   ) {
    rotationMode=BMM_Additive
    translationMode=BMM_Additive
    alphaBoolBlend {blendInTime=0.3, blendOption=Linear, blendOutTime=0.7}
    alphaScaleBiasClamp {bInterpResult=true, interpSpeedDecreasing=5, interpSpeedIncreasing=5}
    boneToModify {boneName=wing_l_master_jnt}
    rotation {yaw=-25}
   }
   n12 = "Transform (Modify) Bone - Bone: wing_r_master_jnt"!(
    ComponentPose=@n11
    Alpha=1
    Translation="0,0,0"
    Rotation_Roll:float=0
    Rotation_Pitch:float=15
    Rotation_Yaw:float=Get rotValue_FlightState_R() -> (rotValue_FlightState_R:double)
    Scale="1,1,1"
   ) -> (Pose) {
    rotationMode=BMM_Additive
    translationMode=BMM_Additive
    alphaBoolBlend {blendInTime=0.3, blendOption=Linear, blendOutTime=0.7}
    alphaScaleBiasClamp {bInterpResult=true, interpSpeedDecreasing=5, interpSpeedIncreasing=5}
    boneToModify {boneName=wing_r_master_jnt}
    rotation {yaw=-15}
   }
   n13 = "Transform (Modify) Bone - Bone: wing_r_master_jnt"(
    ComponentPose=@n12
    Alpha=1
    Translation="0,0,0"
    Rotation_Roll=0
    Rotation_Pitch=0
    Rotation_Yaw="Get Y (Pitch)UpperArmR"() -> ("Y (Pitch)UpperArmR":float)
    Scale="1,1,1"
   ) {
    rotationMode=BMM_Additive
    translationMode=BMM_Additive
    alphaBoolBlend {blendOption=Linear}
    alphaScaleBias {scale=2}
    alphaScaleBiasClamp {bInterpResult=true, scale=0.5}
    boneToModify {boneName=wing_r_master_jnt}
   }
   n14 = "Transform (Modify) Bone - Bone: wing_l_master_jnt"!(ComponentPose=@n13, bAlphaBoolEnabled:bool=True, Translation="0,0,0", Rotation:Rotator="0,-45,0", Scale="1,1,1") -> (Pose) {
    alphaInputType=Bool
    rotationMode=BMM_Additive
    translationMode=BMM_Additive
    alphaBoolBlend {blendInTime=0.2, blendOption=Linear, blendOutTime=0.7}
    alphaScaleBiasClamp {bInterpResult=true}
    boneToModify {boneName=wing_l_master_jnt}
    rotation {yaw=-45}
   }
   n15 = "Transform (Modify) Bone - Bone: wing_r_master_jnt"!(ComponentPose=@n14, bAlphaBoolEnabled:bool=True, Translation="0,0,0", Rotation:Rotator="0,45,0", Scale="1,1,1") -> (Pose) {
    alphaInputType=Bool
    rotationMode=BMM_Additive
    translationMode=BMM_Additive
    alphaBoolBlend {blendInTime=0.2, blendOption=Linear, blendOutTime=0.7}
    alphaScaleBiasClamp {bInterpResult=true}
    boneToModify {boneName=wing_r_master_jnt}
    rotation {yaw=45}
   }
   n16 = Get Para Glide Open Glider_0() -> (Para Glide Open Glider_0:bool)
   n17 = "Transform (Modify) Bone - Bone: wing_l_lower_base_jnt"!(ComponentPose=@n15, bAlphaBoolEnabled:bool=@n16, Translation="0,0,0", Rotation="0,0,30", Scale="1,1,1") -> (Pose) {
    alphaInputType=Bool
    rotationMode=BMM_Additive
    translationMode=BMM_Additive
    alphaBoolBlend {blendInTime=0.5, blendOption=Linear, blendOutTime=0.3}
    alphaScaleBiasClamp {bInterpResult=true, interpSpeedIncreasing=15}
    boneToModify {boneName=wing_l_lower_base_jnt}
    rotation {roll=30}
   }
   n18 = "Transform (Modify) Bone - Bone: wing_l_master_jnt"(ComponentPose=@n17, bAlphaBoolEnabled=@n16, Translation="0,0,0", Rotation="0,-15,5", Scale="1,1,1") {
    alphaInputType=Bool
    rotationMode=BMM_Additive
    translationMode=BMM_Additive
    alphaBoolBlend {blendInTime=0.5, blendOption=Linear, blendOutTime=0.3}
    alphaScaleBiasClamp {bInterpResult=true, interpSpeedIncreasing=15}
    boneToModify {boneName=wing_l_master_jnt}
    rotation {roll=5, yaw=-15}
   }
   n19 = "Transform (Modify) Bone - Bone: wing_l_master_jnt"(ComponentPose=@n18, bAlphaBoolEnabled=True, Translation="0,-1,0", Rotation="0,0,10", Scale="1,1,1") {
    alphaInputType=Bool
    rotationMode=BMM_Additive
    translationMode=BMM_Additive
    alphaBoolBlend {blendInTime=0.5, blendOption=Linear, blendOutTime=0.3}
    alphaScaleBiasClamp {bInterpResult=true, interpSpeedIncreasing=15}
    boneToModify {boneName=wing_l_master_jnt}
    rotation {roll=10}
    translation {y=-1}
   }
   n20 = "Transform (Modify) Bone - Bone: wing_r_lower_base_jnt"!(ComponentPose=@n19, bAlphaBoolEnabled:bool=@n16, Translation="0,0,0", Rotation="0,0,30", Scale="1,1,1") -> (Pose) {
    alphaInputType=Bool
    rotationMode=BMM_Additive
    translationMode=BMM_Additive
    alphaBoolBlend {blendInTime=0.5, blendOption=Linear, blendOutTime=0.3}
    alphaScaleBiasClamp {bInterpResult=true, interpSpeedIncreasing=15}
    boneToModify {boneName=wing_r_lower_base_jnt}
    rotation {roll=30}
   }
   n21 = "Transform (Modify) Bone - Bone: wing_r_master_jnt"(ComponentPose=@n20, bAlphaBoolEnabled=@n16, Translation="0,0,0", Rotation="0,15,5", Scale="1,1,1") {
    alphaInputType=Bool
    rotationMode=BMM_Additive
    translationMode=BMM_Additive
    alphaBoolBlend {blendInTime=0.5, blendOption=Linear, blendOutTime=0.3}
    alphaScaleBiasClamp {bInterpResult=true, interpSpeedIncreasing=15}
    boneToModify {boneName=wing_r_master_jnt}
    rotation {roll=5, yaw=15}
   }
   n22 = "Transform (Modify) Bone - Bone: wing_r_master_jnt"(ComponentPose=@n21, bAlphaBoolEnabled=True, Translation="0,-1,0", Rotation="0,0,10", Scale="1,1,1") {
    alphaInputType=Bool
    rotationMode=BMM_Additive
    translationMode=BMM_Additive
    alphaBoolBlend {blendInTime=0.5, blendOption=Linear, blendOutTime=0.3}
    alphaScaleBiasClamp {bInterpResult=true, interpSpeedIncreasing=15}
    boneToModify {boneName=wing_r_master_jnt}
    rotation {roll=10}
    translation {y=-1}
   }
   n23 = "Transform (Modify) Bone - Bone: wing_l_master_jnt"!(ComponentPose=@n22, Alpha:float=1, Translation="0,0,0", Rotation="0,-10,0", Scale="1,1,1") -> (Pose) {
    rotationMode=BMM_Additive
    translationMode=BMM_Additive
    alphaBoolBlend {blendOption=Linear}
    alphaScaleBiasClamp {bInterpResult=true, interpSpeedDecreasing=5}
    boneToModify {boneName=wing_l_master_jnt}
    rotation {yaw=-10}
   }
   n24 = "Transform (Modify) Bone - Bone: wing_l_master_jnt"(ComponentPose=@n23, Alpha=1, Translation="0,0,0", Rotation="0,10,0", Scale="1,1,1") {
    rotationMode=BMM_Additive
    translationMode=BMM_Additive
    alphaBoolBlend {blendOption=Linear}
    alphaScaleBiasClamp {bInterpResult=true, interpSpeedDecreasing=5}
    boneToModify {boneName=wing_l_master_jnt}
    rotation {yaw=10}
   }
   Output Pose(Result=Blend Poses by bool(BlendPose_0=Component To Local(ComponentPose=@n24), BlendPose_1=Use cached pose 'toTransform'(), bActiveValue=Get bIsSkyDiving()) {blendType=Linear}) {layerGroup=None}
   n25 = RigidBody!(ComponentPose=Local To Component(LocalPose=Use cached pose 'toInput'()), Alpha=0.8, OverrideWorldGravity="0,0,0", ExternalForce) -> (Pose) {
    bClampLinearTranslationLimitToRefPose=true
    bOverrideWorldGravity=true
    overridePhysicsAsset=&M_MED_DustyBun_Pack_Physics
    simulationSpace=BaseBoneSpace
    alphaBoolBlend {blendOption=Linear}
    baseBoneRef {boneName=pack}
    componentAppliedLinearAccClamp=<1000 1000 1000>
    simSpaceSettings {
     dampingAlpha=0
     maxAngularAcceleration=100
     maxAngularVelocity=10
     maxLinearAcceleration=500
     maxLinearVelocity=200
     velocityScaleZ=5
     worldAlpha=1
    }
   }
   n26 = Component To Local(ComponentPose=@n25)
   n27 = RigidBody(ComponentPose=Local To Component(LocalPose=Use cached pose 'toInput'()), OverrideWorldGravity="0,0,0") {
    bClampLinearTranslationLimitToRefPose=true
    bOverrideWorldGravity=true
    overridePhysicsAsset=&M_MED_DustyBun_Pack_Physics
    simulationSpace=BaseBoneSpace
    alphaBoolBlend {blendOption=Linear}
    baseBoneRef {boneName=pack}
    componentAppliedLinearAccClamp=<1000 1000 1000>
    simSpaceSettings {
     dampingAlpha=0
     maxAngularAcceleration=350
     maxAngularVelocity=10
     maxLinearAcceleration=1500
     maxLinearVelocity=1200
     velocityScaleZ=10
     worldAlpha=1
     externalLinearDragV=<1 0.75 0.2>
    }
   }
   n28 = Component To Local(ComponentPose=@n27)
   n29 = Get bIsInFrontEnd()
   n30 = "Transform (Modify) Bone - Bone: Spelvis"!(
    ComponentPose=Local To Component(LocalPose=Blend Poses by bool(BlendPose_0=@n26, BlendPose_1=@n28, bActiveValue=@n29) {blendType=Linear})
    bAlphaBoolEnabled:bool=True
    Translation=Get PelvisLoc() -> (PelvisLoc:Vector)
    Rotation=Get PelvisRot() -> (PelvisRot:Rotator)
    Scale=Get PelvisScale() -> (PelvisScale:Vector)
   ) -> (Pose) {
    alphaInputType=Bool
    rotationMode=BMM_Replace
    rotationSpace=BCS_WorldSpace
    scaleMode=BMM_Replace
    scaleSpace=BCS_WorldSpace
    translationMode=BMM_Replace
    translationSpace=BCS_WorldSpace
    alphaBoolBlend {blendInTime=0.2, blendOption=Linear, blendOutTime=0.2}
    boneToModify {boneName=Spelvis}
    rotation {roll=25}
   }
   n31 = Get forwardMov_fvalue() -> (forwardMov_fvalue:double)
   n32 = "Transform (Modify) Bone - Bone: wing_l_master_jnt"(ComponentPose=@n30, Alpha=@n31, Translation="0,0,0", Rotation="4.291,-35,-4.3", Scale="1,1,1") {
    rotationMode=BMM_Additive
    translationMode=BMM_Additive
    alphaBoolBlend {blendOption=Linear}
    alphaScaleBiasClamp {bInterpResult=true, interpSpeedDecreasing=5}
    boneToModify {boneName=wing_l_master_jnt}
    rotation=<rot 4.291 -35 -4.3>
   }
   n33 = "Transform (Modify) Bone - Bone: wing_r_master_jnt"!(ComponentPose=@n32, Alpha:float=@n31, Translation="0,0,0", Rotation="-4.168,35,-4.175", Scale="1,1,1") -> (Pose) {
    rotationMode=BMM_Additive
    translationMode=BMM_Additive
    alphaBoolBlend {blendOption=Linear}
    alphaScaleBiasClamp {bInterpResult=true, interpSpeedDecreasing=5}
    boneToModify {boneName=wing_r_master_jnt}
    rotation=<rot -4.168 35 -4.175>
   }
   n34 = Get leftMov_fvalue() -> (leftMov_fvalue:double)
   n35 = "Transform (Modify) Bone - Bone: wing_l_master_jnt"(ComponentPose=@n33, Alpha=@n34, Translation="0,0,0", Rotation="0,-43,0", Scale="1,1,1") {
    rotationMode=BMM_Additive
    translationMode=BMM_Additive
    alphaBoolBlend {blendOption=Linear}
    alphaScaleBiasClamp {bInterpResult=true, interpSpeedIncreasing=8}
    boneToModify {boneName=wing_l_master_jnt}
    rotation {yaw=-43}
   }
   n36 = "Transform (Modify) Bone - Bone: wing_r_master_jnt"(ComponentPose=@n35, Alpha=@n34, Translation="0,0,0", Rotation="0,-17,0", Scale="1,1,1") {
    rotationMode=BMM_Additive
    translationMode=BMM_Additive
    alphaBoolBlend {blendOption=Linear}
    alphaScaleBiasClamp {bInterpResult=true, interpSpeedIncreasing=8}
    boneToModify {boneName=wing_r_master_jnt}
    rotation {yaw=-17}
   }
   n37 = "Transform (Modify) Bone - Bone: wing_l_lower_base_jnt"!(ComponentPose=@n36, Alpha:float=@n34, Translation="0,0,0", Rotation="0,0,0", Scale="1,1,1") -> (Pose) {
    rotationMode=BMM_Additive
    translationMode=BMM_Additive
    alphaBoolBlend {blendOption=Linear}
    alphaScaleBiasClamp {bInterpResult=true}
    boneToModify {boneName=wing_l_lower_base_jnt}
   }
   n38 = "Transform (Modify) Bone - Bone: wing_l_ponytail_02_jnt"(ComponentPose=@n37, Alpha=@n34, Translation="0,0,0", Rotation="0,0,0", Scale="1,1,1") {
    rotationMode=BMM_Additive
    translationMode=BMM_Additive
    alphaBoolBlend {blendOption=Linear}
    alphaScaleBiasClamp {bInterpResult=true}
    boneToModify {boneName=wing_l_ponytail_02_jnt}
   }
   n39 = Get RightMov_fvalue() -> (RightMov_fvalue:double)
   n40 = "Transform (Modify) Bone - Bone: wing_l_master_jnt"(ComponentPose=@n38, Alpha=@n39, Translation="0,0,0", Rotation="0,17,0", Scale="1,1,1") {
    rotationMode=BMM_Additive
    translationMode=BMM_Additive
    alphaBoolBlend {blendOption=Linear}
    alphaScaleBiasClamp {bInterpResult=true, interpSpeedIncreasing=8}
    boneToModify {boneName=wing_l_master_jnt}
    rotation {yaw=17}
   }
   n41 = "Transform (Modify) Bone - Bone: wing_r_master_jnt"(ComponentPose=@n40, Alpha=@n39, Translation="0,0,0", Rotation="0,40,0", Scale="1,1,1") {
    rotationMode=BMM_Additive
    translationMode=BMM_Additive
    alphaBoolBlend {blendOption=Linear}
    alphaScaleBiasClamp {bInterpResult=true, interpSpeedIncreasing=8}
    boneToModify {boneName=wing_r_master_jnt}
    rotation {yaw=40}
   }
   n42 = "Transform (Modify) Bone - Bone: wing_r_lower_base_jnt"!(ComponentPose=@n41, Alpha:float=@n39, Translation="0,0,0", Rotation="0,0,0", Scale="1,1,1") -> (Pose) {
    rotationMode=BMM_Additive
    translationMode=BMM_Additive
    alphaBoolBlend {blendOption=Linear}
    alphaScaleBiasClamp {bInterpResult=true}
    boneToModify {boneName=wing_r_lower_base_jnt}
   }
   n43 = "Transform (Modify) Bone - Bone: wing_r_ponytail_02_jnt"(ComponentPose=@n42, Alpha=@n39, Translation="0,0,0", Rotation="0,0,0", Scale="1,1,1") {
    rotationMode=BMM_Additive
    translationMode=BMM_Additive
    alphaBoolBlend {blendOption=Linear}
    alphaScaleBiasClamp {bInterpResult=true}
    boneToModify {boneName=wing_r_ponytail_02_jnt}
   }
   Save cached pose 'toTransform'(Pose=Component To Local(ComponentPose=@n43)) {cacheName=toTransform}
   n44 = "Transform (Modify) Bone - Bone: wing_l_master_jnt"(
    ComponentPose=Local To Component(LocalPose=Use cached pose 'toTransform'())
    Alpha=1
    Translation="0,-1,0"
    Rotation="-5,20,0"
    Scale="1,1,1"
   ) {
    rotationMode=BMM_Additive
    translationMode=BMM_Additive
    alphaBoolBlend {blendOption=Linear}
    alphaScaleBiasClamp {bInterpResult=true, interpSpeedDecreasing=5}
    boneToModify {boneName=wing_l_master_jnt}
    rotation {pitch=-5, yaw=20}
    translation {y=-1}
   }
   n45 = "Transform (Modify) Bone - Bone: wing_r_master_jnt"(ComponentPose=@n44, Alpha=1, Translation="0,-1,0", Rotation="5,-20,0", Scale="1,1,1") {
    rotationMode=BMM_Additive
    translationMode=BMM_Additive
    alphaBoolBlend {blendOption=Linear}
    alphaScaleBiasClamp {bInterpResult=true, interpSpeedDecreasing=5}
    boneToModify {boneName=wing_r_master_jnt}
    rotation {pitch=5, yaw=-20}
    translation {y=-1}
   }
   n46 = "Transform (Modify) Bone - Bone: wing_l_mid_base_jnt"(ComponentPose=@n45, Alpha=1, Translation="1,0,1", Rotation="10,5,4", Scale="1,1,1") {
    rotationMode=BMM_Additive
    translationMode=BMM_Additive
    alphaBoolBlend {blendOption=Linear}
    alphaScaleBiasClamp {bInterpResult=true, interpSpeedDecreasing=5}
    boneToModify {boneName=wing_l_mid_base_jnt}
    rotation=<rot 10 5 4>
    translation {x=1, z=1}
   }
   n47 = "Transform (Modify) Bone - Bone: wing_r_mid_base_jnt"(ComponentPose=@n46, Alpha=1, Translation="-1,0,1", Rotation="-10,-5,4", Scale="1,1,1") {
    rotationMode=BMM_Additive
    translationMode=BMM_Additive
    alphaBoolBlend {blendOption=Linear}
    alphaScaleBiasClamp {bInterpResult=true, interpSpeedDecreasing=5}
    boneToModify {boneName=wing_r_mid_base_jnt}
    rotation=<rot -10 -5 4>
    translation {x=-1, z=1}
   }
   n48 = "Transform (Modify) Bone - Bone: wing_l_mid_base_jnt"(ComponentPose=@n47, Alpha=1, Translation="1,0,0", Rotation="4,-5,0", Scale="1,1,1") {
    rotationMode=BMM_Additive
    translationMode=BMM_Additive
    alphaBoolBlend {blendOption=Linear}
    alphaScaleBiasClamp {bInterpResult=true, interpSpeedDecreasing=5}
    boneToModify {boneName=wing_l_mid_base_jnt}
    rotation {pitch=4, yaw=-5}
    translation {x=1}
   }
   n49 = "Transform (Modify) Bone - Bone: wing_r_mid_base_jnt"(ComponentPose=@n48, Alpha=1, Translation="-1,0,0", Rotation="-4,5,0", Scale="1,1,1") {
    rotationMode=BMM_Additive
    translationMode=BMM_Additive
    alphaBoolBlend {blendOption=Linear}
    alphaScaleBiasClamp {bInterpResult=true, interpSpeedDecreasing=5}
    boneToModify {boneName=wing_r_mid_base_jnt}
    rotation {pitch=-4, yaw=5}
    translation {x=-1}
   }
   n50 = "Transform (Modify) Bone - Bone: wing_l_lower_base_jnt"(ComponentPose=@n49, Alpha=1, Translation="0,0,0", Rotation="-5,0,0", Scale="1,1,1") {
    rotationMode=BMM_Additive
    translationMode=BMM_Additive
    alphaBoolBlend {blendOption=Linear}
    alphaScaleBiasClamp {bInterpResult=true, interpSpeedDecreasing=5}
    boneToModify {boneName=wing_l_lower_base_jnt}
    rotation {pitch=-5}
   }
   n51 = "Transform (Modify) Bone - Bone: wing_r_lower_base_jnt"(ComponentPose=@n50, Alpha=1, Translation="0,0,0", Rotation="5,0,0", Scale="1,1,1") {
    rotationMode=BMM_Additive
    translationMode=BMM_Additive
    alphaBoolBlend {blendOption=Linear}
    alphaScaleBiasClamp {bInterpResult=true, interpSpeedDecreasing=5}
    boneToModify {boneName=wing_r_lower_base_jnt}
    rotation {pitch=5}
   }
   n52 = Get backMovementAlpha() -> (backMovementAlpha:double)
   n53 = "Transform (Modify) Bone - Bone: wing_l_master_jnt"(ComponentPose=@n51, Alpha=@n52, Translation="0,0,0", Rotation="10,-25,0", Scale="1,1,1") {
    rotationMode=BMM_Additive
    translationMode=BMM_Additive
    alphaBoolBlend {blendInTime=0.5, blendOption=Linear, blendOutTime=0.3}
    alphaScaleBiasClamp {bInterpResult=true, interpSpeedDecreasing=7, interpSpeedIncreasing=3}
    boneToModify {boneName=wing_l_master_jnt}
    rotation {pitch=10, yaw=-25}
   }
   n54 = "Transform (Modify) Bone - Bone: wing_l_ponytail_02_jnt"(ComponentPose=@n53, Alpha=@n52, Translation="0,0,0", Rotation="0,0,40", Scale="1,1,1") {
    rotationMode=BMM_Additive
    translationMode=BMM_Additive
    alphaBoolBlend {blendOption=Linear}
    alphaScaleBiasClamp {bInterpResult=true, interpSpeedDecreasing=7, interpSpeedIncreasing=3}
    boneToModify {boneName=wing_l_ponytail_02_jnt}
    rotation {roll=40}
   }
   n55 = "Transform (Modify) Bone - Bone: wing_r_master_jnt"(ComponentPose=@n54, Alpha=@n52, Translation="0,0,0", Rotation="-10,25,0", Scale="1,1,1") {
    rotationMode=BMM_Additive
    translationMode=BMM_Additive
    alphaBoolBlend {blendInTime=0.5, blendOption=Linear, blendOutTime=0.3}
    alphaScaleBiasClamp {bInterpResult=true, interpSpeedDecreasing=7, interpSpeedIncreasing=3}
    boneToModify {boneName=wing_r_master_jnt}
    rotation {pitch=-10, yaw=25}
   }
   n56 = "Transform (Modify) Bone - Bone: wing_r_ponytail_02_jnt"(ComponentPose=@n55, Alpha=@n52, Translation="0,0,0", Rotation="0,0,40", Scale="1,1,1") {
    rotationMode=BMM_Additive
    translationMode=BMM_Additive
    alphaBoolBlend {blendOption=Linear}
    alphaScaleBiasClamp {bInterpResult=true, interpSpeedDecreasing=7, interpSpeedIncreasing=3}
    boneToModify {boneName=wing_r_ponytail_02_jnt}
    rotation {roll=40}
   }
   n57 = "Transform (Modify) Bone - Bone: wing_l_master_jnt"(ComponentPose=@n56, Alpha=1, Translation="0,0,0", Rotation="0,-31,0", Scale="1,1,1") {
    rotationMode=BMM_Additive
    translationMode=BMM_Additive
    alphaBoolBlend {blendOption=Linear}
    alphaScaleBiasClamp {bInterpResult=true, interpSpeedDecreasing=5}
    boneToModify {boneName=wing_l_master_jnt}
    rotation {yaw=-31}
   }
   n58 = "Transform (Modify) Bone - Bone: wing_r_master_jnt"(ComponentPose=@n57, Alpha=1, Translation="0,0,0", Rotation="0,31,0", Scale="1,1,1") {
    rotationMode=BMM_Additive
    translationMode=BMM_Additive
    alphaBoolBlend {blendOption=Linear}
    alphaScaleBiasClamp {bInterpResult=true, interpSpeedDecreasing=5}
    boneToModify {boneName=wing_r_master_jnt}
    rotation {yaw=31}
   }
   n59 = Get backMovementAlpha_forCrouch() -> (backMovementAlpha_forCrouch:double)
   n60 = "Transform (Modify) Bone - Bone: wing_l_master_jnt"(ComponentPose=@n58, Alpha=@n59, Translation="0,0,0", Rotation="0,0,-20", Scale="1,1,1") {
    rotationMode=BMM_Additive
    translationMode=BMM_Additive
    alphaBoolBlend {blendOption=Linear}
    alphaScaleBiasClamp {bInterpResult=true, interpSpeedDecreasing=5, interpSpeedIncreasing=5}
    boneToModify {boneName=wing_l_master_jnt}
    rotation {roll=-20}
   }
   n61 = "Transform (Modify) Bone - Bone: wing_r_master_jnt"(ComponentPose=@n60, Alpha=@n59, Translation="0,0,0", Rotation="0,0,-20", Scale="1,1,1") {
    rotationMode=BMM_Additive
    translationMode=BMM_Additive
    alphaBoolBlend {blendOption=Linear}
    alphaScaleBiasClamp {bInterpResult=true, interpSpeedDecreasing=5, interpSpeedIncreasing=5}
    boneToModify {boneName=wing_r_master_jnt}
    rotation {roll=-20}
   }
   n62 = "Transform (Modify) Bone - Bone: pack"!(
    ComponentPose=@n61
    bAlphaBoolEnabled:bool=Get bIsPlayingMeleeAnim()
    Translation="0,0,0"
    Rotation=Get rawRotationSpine() -> (rawRotationSpine:Rotator)
    Scale
   ) -> (Pose) {
    alphaInputType=Bool
    rotationMode=BMM_Additive
    translationMode=BMM_Additive
    alphaBoolBlend {blendInTime=0.2, blendOption=Linear, blendOutTime=0.3}
    boneToModify {boneName=pack}
   }
   Save cached pose 'CombAnimCache'(Pose=Component To Local(ComponentPose=@n62)) {cacheName=CombAnimCache}
   n63 = "Transform (Modify) Bone - Bone: pack"!(ComponentPose=Local To Component(LocalPose=Input Pose()), Alpha:float, Translation="0,4.5,0", Rotation="0.000,0.000,10.000", Scale) -> (Pose) {
    rotationMode=BMM_Additive
    translationMode=BMM_Additive
    alphaBoolBlend {blendOption=Linear}
    boneToModify {boneName=pack}
    rotation {roll=10}
    translation {y=4.5}
   }
   Save cached pose 'toInput'(Pose=Component To Local(ComponentPose=@n63)) {cacheName=toInput}
  ]}
 EventGraph {nodes [
   n1 = Event BlueprintUpdateAnimation()
   > Set Delta Time X(Delta Time X=@n1.DeltaTimeX)
   > n2 = PlayerMovement(AnimInstance=Self-Reference(), XDotProductCoefficient=1, YDotProductCoefficient=1, ZDotProductCoefficient=1)
   > n3 = Sequence!(execute) -> (then_0, then_1, then_2:Exec, then_3:Exec, then_4:Exec, then_5:Exec, then_6:Exec, then_7:Exec, then_8:Exec)
   n4 = GetDeltaTransformFromRefPose!(self=GetOwnerSkeletalMeshComponent(), BoneName=spine_05, BaseName=spine_01) -> (ReturnValue_Location:Vector, ReturnValue_Rotation:Rotator, ReturnValue_Scale:Vector)
   Set Parashuting(execute=@n3.then_8, Parashuting:double=MapRangeClamped(Value=BreakRotator(InRot=@n4.ReturnValue_Rotation).Yaw, InRangeA=-25, InRangeB=-65)) -> (Output_Get:double)
   n5 = GetDeltaTransformFromRefPose(self=GetOwnerSkeletalMeshComponent(), BoneName=pelvis, BaseName=root)
   > Set Skydive(Skydive:double=MapRangeClamped(Value=BreakRotator(InRot=@n5.ReturnValue_Rotation).Yaw, InRangeA=20, InRangeB=80)) -> (Output_Get:double)
   n6 = Get bIsCrouching()
   n7 = "Absolute (Float)"(A="Min (Float)"(A:double=@n2.YMovement, B:double=0) -> (ReturnValue:double) {is=title K2Node_CommutativeAssociativeBinaryOperator})
   Set backMovementAlpha_forCrouch(execute=@n3.then_6, backMovementAlpha_forCrouch:double=SelectFloat(A=@n7, B=0, bPickA=@n6)) -> (Output_Get:double)
   n8 = Get bIsTargeting()
   n9 = NOT Boolean(A=Get bIsJumping())
   n10 = NOT Boolean(A=Get bIsCrouching())
   Set condMovement(execute=@n3.then_3, condMovement:bool=AND Boolean(A=AND Boolean(A=@n9, B=@n10), B=NOT Boolean(A=@n8))) -> (Output_Get:bool)
   n11 = NOT Boolean(A=Get bIsSkydiveDiveMode())
   n12 = NOT Boolean(A=Get bIsSkydiveFloating())
   n13 = GetAlphaFromReferenceTransformDelta(SkeletalMeshComponent=GetOwnerSkeletalMeshComponent(), BoneName=spine_05, BaseName=spine_04, InputRange=(X=0,Y=1), TwistAxis="0,0,1")
   > Branch(Condition=AND Boolean(A="float < float"(A=@n13.RawOutput, B=-3), B=AND Boolean(A=@n11, B=@n12)))
   n14 = MakeRotator!(Roll, Pitch, Yaw=@n13.RawOutput) -> (ReturnValue_Roll:float, ReturnValue_Pitch:float, ReturnValue_Yaw:float)
   > Set rawRotationSpine(rawRotationSpine:Rotator=MakeRotator!(Roll, Pitch, Yaw=float * float(A=@n14.ReturnValue_Yaw, B=1.9) {is=title K2Node_PromotableOperator}) -> (ReturnValue:Rotator)) -> (Output_Get:Rotator)
   n15 = Get bIsSkyDiving()
   n16 = NOT Boolean(A=Get bIsJumping())
   n17 = "To Boolean (Integer)"!(InInt:int=Round(A="Max (Float)"(A=@n2, B=0)) {is=title K2Node_CallFunction}) -> (ReturnValue:bool)
   n18 = Branch(execute=@n3.then_1, Condition=AND Boolean(A=AND Boolean(A=@n17, B=@n16), B=NOT Boolean(A=@n15)))
   n19 = Get bIsSkyDiving()
   > n20 = Branch(Condition=AND Boolean(A=Get bIsCrouching(), B=NOT Boolean(A=@n19)))
   n21 = Set RightMov_fvalue(execute=@n18.else, RightMov_fvalue:double=0) -> (Output_Get:double)
   n22 = Get Delta Time X()
   n23 = MapRangeClamped(Value=FInterpTo(Current=@n21, Target=float + float(A=@n21, B=1), DeltaTime=@n22, InterpSpeed=4), InRangeA=0, InRangeB=1)
   n24 = Set RightMov_fvalue(execute=@n20.else, RightMov_fvalue=@n23)
   Set RightMov_fvalue(execute=@n20.then, RightMov_fvalue=FInterpTo(Current=@n24, Target=float * float(A=@n23, B=1), DeltaTime=Get Delta Time X(), InterpSpeed=25))
   n25 = Get bIsSkyDiving()
   n26 = "Min (Float)"(A=@n2, B=0)
   n27 = "To Boolean (Integer)"(InInt=Round(A="Absolute (Float)"(A=@n26)))
   n28 = NOT Boolean(A=Get bIsJumping())
   n29 = Branch(execute=@n3.then_2, Condition=AND Boolean(A=AND Boolean(A=@n27, B=@n28), B=NOT Boolean(A=@n25)))
   > n30 = Branch(Condition=Get bIsCrouching())
   n31 = Set leftMov_fvalue(execute=@n29.else, leftMov_fvalue:double=0) -> (Output_Get:double)
   n32 = MapRangeClamped(Value=FInterpTo(Current=@n31, Target=float + float(A=@n31, B=1), DeltaTime=@n22, InterpSpeed=4), InRangeA=0, InRangeB=1)
   n33 = Set leftMov_fvalue(execute=@n30.else, leftMov_fvalue=@n32)
   Set leftMov_fvalue(execute=@n30.then, leftMov_fvalue=FInterpTo(Current=@n33, Target=float * float(A=@n32, B=1), DeltaTime=Get Delta Time X(), InterpSpeed=25))
   n34 = NOT Boolean(A=@n6)
   n35 = NOT Boolean(A=Get bIsSkyDiving())
   Set backMovementAlpha(execute=@n3.then_5, backMovementAlpha:double=SelectFloat(A=@n7, B=0, bPickA=AND Boolean(A=@n35, B=@n34))) -> (Output_Get:double)
   n36 = Get bIsJumping()
   n37 = "To Boolean (Integer)"(InInt=Round(A="Max (Float)"(A=@n2.YMovement, B=0)))
   n38 = Get condMovement() -> (condMovement:bool)
   n39 = Branch(execute=@n3.then_4, Condition=AND Boolean(A=AND Boolean(A=@n37, B=@n38), B=NOT Boolean(A=@n36)))
   > n40 = Branch(Condition=Get bIsTacticalSprinting())
   n41 = Set forwardMov_fvalue(execute=@n39.else, forwardMov_fvalue:double=0) -> (Output_Get:double)
   n42 = MapRangeClamped(
    Value=FInterpTo(Current=@n41, Target=float + float(A=@n41, B=1), DeltaTime=Get Delta Time X(), InterpSpeed=22)
    InRangeA=0
    InRangeB=1
    OutRangeB=0.5
   )
   n43 = Set forwardMov_fvalue(execute=@n40.else, forwardMov_fvalue=@n42)
   Set forwardMov_fvalue(
    execute=@n40.then
    forwardMov_fvalue=FInterpTo(Current=@n43, Target=float + float(A=@n42, B=0.5), DeltaTime=Get Delta Time X(), InterpSpeed=10)
   )
   n44 = Get bIsCrouching()
   Set newCrouching(execute=@n3.then_0, newCrouching:bool=OR Boolean(A=Get bIsTargeting(), B=@n44)) -> (Output_Get:bool)
   n45 = Round(A=Get backMovementAlpha())
   > Set forCrouchingBack(forCrouchingBack:bool=AND Boolean(A="To Boolean (Integer)"(InInt=@n45), B=@n44)) -> (Output_Get:bool)
   n46 = "float > float"(A="Max (Float)"(A=Get PawnVelocityZ(), B=0), B=3000)
   n47 = "Not Equal (Boolean)"(A=Get bIsParachuteOpened(), B=true) -> (ReturnValue=~)
   > Set bGreaterVel(bGreaterVel:bool=AND Boolean(A=AND Boolean(A=@n46, B=@n47), B=Get bIsSkyDiving())) -> (Output_Get:bool)
   n48 = GetPartSkeletalMeshComponent()
   > Is Valid(InputObject=@n48)
   n49 = GetAttachParent!(self:Scene Component Object Reference=@n48) -> (ReturnValue:Scene Component Object Reference)
   > Is Valid(InputObject=@n49)
   n50 = GetSocketTransform!(self=@n49, InSocketName=pelvis, TransformSpace=RTS_World) -> (ReturnValue_Location:Vector, ReturnValue_Rotation:Rotator, ReturnValue_Scale:Vector)
   > Set PelvisLoc(PelvisLoc:Vector=@n50) -> (Output_Get:Vector)
   > Set PelvisRot(PelvisRot:Rotator=@n50.ReturnValue_Rotation) -> (Output_Get:Rotator)
   > Set PelvisScale(PelvisScale:Vector=@n50.ReturnValue_Scale) -> (Output_Get:Vector)
   n51 = Sequence!(execute=@n3.then_7) -> (then_0, then_1, then_2)
   n52 = GetOwnerSkeletalMeshComponent()
   n53 = BreakRotator(InRot=GetSocketTransform(self=@n52, InSocketName=upperarm_l, TransformSpace=RTS_Component).ReturnValue_Rotation)
   > Set Return Value Rotation l(Return Value Rotation l:Rotator=MakeRotator(Yaw="Clamp (Float)"(Value=@n53.Yaw, Min=-15, Max=5))) -> (Output_Get:Rotator)
   > n54 = Negate Float!(" ":Exec, Value:"Float (double-precision) (by ref)"=float / float(A=@n53.Pitch, B=1.3)) -> ("  ":Exec, Result:double=~) {is=title K2Node_MacroInstance}
   > "Set Y (Pitch)UpperArmL"("Y (Pitch)UpperArmL":float=@n54) -> (Output_Get:float)
   n55 = BreakRotator(InRot=GetSocketTransform(self=@n52, InSocketName=upperarm_r, TransformSpace=RTS_Component).ReturnValue_Rotation)
   n56 = "Clamp (Float)"(Value=@n55.Yaw, Min=5, Max=15)
   > Set Return Value Rotation_r(Return Value Rotation_r:Rotator=MakeRotator(Yaw=float * float(A=@n56, B=4))) -> (Output_Get:Rotator)
   > n57 = Negate Float(Value=float / float(A=@n55.Pitch, B=3)) -> (Result=~)
   > "Set Y (Pitch)UpperArmR"("Y (Pitch)UpperArmR":float=@n57) -> (Output_Get:float)
   > n58 = Branch(Condition=Get bIsInFrontEnd())
   > Set Component Linear Acc Scale(Component Linear Acc Scale:Vector="0.8,0.8,1") -> (Output_Get:Vector)
   > Set Para Glide Open Glider_0(Para Glide Open Glider_0:bool=false) -> (Output_Get:bool)
   n59 = Set fort_player(
    execute=@n58.else
    fort_player:Fort Player Anim Instance Object Reference=GetOwnerFortPlayerAnimInstance!(self:Custom Character Part Anim Instance Object Reference) -> (ReturnValue:Fort Player Anim Instance Object Reference)
   ) -> (Output_Get:Fort Player Anim Instance Object Reference)
   > n60 = Set Para Glide Open Glider(Para Glide Open Glider:bool=Get bParaGlide_OpenGlider!(self:Fort Player Anim Instance Object Reference=@n59) -> (bParaGlide_OpenGlider:bool)) -> (Output_Get:bool)
   > n61 = Branch(Condition=@n60)
   > Set Override World Gravity(Override World Gravity:Vector="0,0,-200") -> (Output_Get:Vector)
   > Set Component Linear Acc Scale(Component Linear Acc Scale="0, 0, 0")
   > Delay(Duration=0.3)
   > Set Para Glide Open Glider_0(Para Glide Open Glider_0=true)
   Set Component Linear Acc Scale(execute=@n61.else, Component Linear Acc Scale="0.8,0.8,1")
   > Delay(Duration=1.5)
   > Set Para Glide Open Glider_0(Para Glide Open Glider_0=false)
   Branch(execute=@n51.then_1, Condition=Get bIsSkydiveFloating())
   n62 = GetSocketTransform!(self=GetOwnerSkeletalMeshComponent(), InSocketName=spine_05, TransformSpace=RTS_Component) -> (
    ReturnValue_Location_X:double
    ReturnValue_Location_Y:double
    ReturnValue_Location_Z:double
    ReturnValue_Rotation_Roll:float
    ReturnValue_Rotation_Pitch:float
    ReturnValue_Rotation_Yaw:float
    ReturnValue_Scale
   )
   > Set rotValue_FlightState_L(rotValue_FlightState_L:double="Clamp (Float)"(Value=@n62, Min=-45, Max=5)) -> (Output_Get:double)
   > Set rotValue_FlightState_R(rotValue_FlightState_R:double="Clamp (Float)"(Value=@n62, Min=-15, Max=40)) -> (Output_Get:double)
   > "Set Return ValueSpine_05 Rotation Z (Yaw)"("Return ValueSpine_05 Rotation Z (Yaw)":float=MapRangeClamped(Value=@n62.ReturnValue_Rotation_Yaw, InRangeA=75, InRangeB=96, OutRangeA=7, OutRangeB=15)) -> (Output_Get:float)
   n63 = Branch(execute=@n51.then_2, Condition=Get bIsSkyDiving())
   > Set alphaSkyCond(alphaSkyCond:double=0.3) -> (Output_Get:double)
   > Set Override World Gravity(Override World Gravity="0,0,-260")
   Set alphaSkyCond(execute=@n63.else, alphaSkyCond=0.8)
   > Set Override World Gravity(Override World Gravity="0,0,-980")
   TryGetPawnOwner()
  ]}
]
variables [
 alphaSkyCond {type=number}
 backMovementAlpha {type=number}
 backMovementAlpha_forCrouch {type=number}
 bGreaterVel {type=boolean}
 Component Linear Acc Scale {type=Vector}
 condMovement {type=boolean}
 Delta Time X {type=number}
 forCrouchingBack {type=boolean}
 fort_player {type=FortPlayerAnimInstance, default=null}
 forwardMov_fvalue {type=number}
 leftMov_fvalue {type=number}
 newCrouching {type=boolean}
 Override World Gravity {type=Vector}
 Para Glide Open Glider {type=boolean}
 Para Glide Open Glider_0 {type=boolean}
 Parashuting {type=number}
 PelvisLoc {type=Vector}
 PelvisRot {type=Rotator}
 PelvisScale {type=Vector}
 rawRotationSpine {type=Rotator}
 Return Value Rotation l {type=Rotator}
 Return Value Rotation_r {type=Rotator}
 "Return ValueSpine_05 Rotation Z (Yaw)" {type=number}
 RightMov_fvalue {type=number}
 rotValue_FlightState_L {type=number}
 rotValue_FlightState_R {type=number}
 Skydive {type=number}
 "Y (Pitch)UpperArmL" {type=number}
 "Y (Pitch)UpperArmR" {type=number}
]
